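; BP_S36_FountainGate is a Season 36 quest actor for a fountain gate portal that visually opens when the player interacts with it.
; Its side-effect event runs only on remote clients and switches on the trigger tag, handling only tag C, the interaction stage.
; That tag sets the PortalOpen parameter on the portal's Niagara component and spawns a one-shot interaction effect at its transform.
; BeginPlay, overlap and Tick are unused stubs.
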(event EventBeginPlay
  (|CallParentFunction) ; This node is disabled and will not be called.\nDrag off pins to build functionality.
)

(event Collision|EventActorBeginOverlap
  (params (OtherActor :type Actor%20Object%20Reference))
  (|CallParentFunction OtherActor) ; This node is disabled and will not be called.\nDrag off pins to build functionality.
)

(event EventTick
  (params (DeltaSeconds :type Float%20%28single-precision%29))
  (|CallParentFunction DeltaSeconds) ; This node is disabled and will not be called.\nDrag off pins to build functionality.
)

(event SideEffects|EventBP_ApplySideEffect
  (params (ReactionSource :type Object%20Reference) (TargetObject :type Object%20Reference) (SideEffectTriggerTag :type Gameplay%20Tag%20Structure%20%28by%20ref%29) (SideEffectTriggerType :type EBlueprintableSideEffectTriggerType%20Enum%20%28by%20ref%29 :default "TriggeredAsTarget"))
  (bind (_returnvalue_location _returnvalue_rotation) (Math|Transform|BreakTransform (Transformation|GetWorldTransform (Variables|Default|GetNS_Utopia_Sprite_Portal))))
  (Networking|SwitchHasAuthority
    (:Remote
      ;[0.1264 0.15 0.135 0.5] A - default state, no interaction
      ; B - awaiting player interaction
      ; C - Fires when player interacts (portal opens, etc)
      (switch Utilities|FlowControl|Switch|SwitchonGameplayTag SideEffectTriggerTag
        (:Quests.S36.SDA.BPSideEffect.C
          (Effects|Components|ParticleSystem|SetBoolParameter (Variables|Default|GetNS_Utopia_Sprite_Portal) "PortalOpen" true)
          (Niagara|SpawnSystematLocation "/Hermes/Environments/FX/Portal/NS_Utopia_Sprite_Portal_Interaction.NS_Utopia_Sprite_Portal_Interaction" _returnvalue_location _returnvalue_rotation))))))

(fn ConstructionScript
  (|CallParentFunction))
