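; BP_SDBGA_S36_SpireRemnant is a Season 36 quest actor, the Spire Remnant, whose static mesh reacts to gameplay-tagged side effects.
; Its side effect event switches on the trigger tag, printing a debug string for each case.
; Tag B starts a timeline that spins the tilted mesh while bobbing it vertically relative to the capsule.
; Tag C hides the mesh; the BeginPlay, overlap and Tick events are unused stubs.
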(event EventBeginPlay
  (|CallParentFunction) ; This node is disabled and will not be called.\nDrag off pins to build functionality.
)

(event Collision|EventActorBeginOverlap
  (params (OtherActor :type Actor%20Object%20Reference))
  (|CallParentFunction OtherActor) ; This node is disabled and will not be called.\nDrag off pins to build functionality.
)

(event EventTick
  (params (DeltaSeconds :type Float%20%28single-precision%29))
  (|CallParentFunction DeltaSeconds) ; This node is disabled and will not be called.\nDrag off pins to build functionality.
)

(event SideEffects|EventBP_ApplySideEffect
  (params (ReactionSource :type Object%20Reference) (TargetObject :type Object%20Reference) (SideEffectTriggerTag :type Gameplay%20Tag%20Structure%20%28by%20ref%29) (SideEffectTriggerType :type EBlueprintableSideEffectTriggerType%20Enum%20%28by%20ref%29 :default "TriggeredAsTarget"))
  (switch Utilities|FlowControl|Switch|SwitchonGameplayTag SideEffectTriggerTag
    (:Quests.S36.SDA.BPSideEffect.B
      (Development|PrintString "BPSIDE EFFECT B" true true "(R=0.000000,G=0.057936,B=1.000000,A=1.000000)")
      (CallFunction|StartSMRotator))
    (:Quests.S36.SDA.BPSideEffect.C
      (Development|PrintString "BPSIDE EFFECT C" true true "(R=0.159137,G=1.000000,B=0.000000,A=1.000000)")
      (Rendering|SetVisibility (Variables|Default|GetStaticMesh) false false))))

(event Custom|StartSMRotator
  (bind (_direction _rotation _bob) (Timeline|Timeline_0 PlayFromStart)
    (:Update
      (Transformation|SetWorldRotation (Variables|Default|GetStaticMesh) (Math|Rotator|MakeRotator 6e+01 0.0 (Math|Float|Lerp 0.0 3.6e+02 _rotation)) false false)
      (Transformation|SetRelativeLocation (Variables|Default|GetStaticMesh) (Math|Vector|MakeVector 0.0 0.0 (+ (+ (- (.z (Transformation|GetWorldLocation (Variables|Default|GetCapsule))) (.z (Transformation|GetActorLocation))) (Math|Float|Lerp 5.0 -5.0 _bob)) 32.0)) false false))))

(fn ConstructionScript
  (|CallParentFunction))
