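; Shard sign quest actor for Battle Pass S36 that plays Niagara effects in response to a blueprint side effect.
; When the side effect carries the Quests.S36.SDA.BPSideEffect.B tag, it spawns a burst system attached above the root and activates the idle Niagara component.
; This runs only on the Remote branch of an authority switch, so instances with authority skip the visuals.
; BeginPlay, overlap, tick and construction add no logic.
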(event EventBeginPlay
  (|CallParentFunction) ; This node is disabled and will not be called.\nDrag off pins to build functionality.
)

(event Collision|EventActorBeginOverlap
  (params (OtherActor :type Actor%20Object%20Reference))
  (|CallParentFunction OtherActor) ; This node is disabled and will not be called.\nDrag off pins to build functionality.
)

(event EventTick
  (params (DeltaSeconds :type Float%20%28single-precision%29))
  (|CallParentFunction DeltaSeconds) ; This node is disabled and will not be called.\nDrag off pins to build functionality.
)

(event SideEffects|EventBP_ApplySideEffect
  (params (ReactionSource :type Object%20Reference) (TargetObject :type Object%20Reference) (SideEffectTriggerTag :type Gameplay%20Tag%20Structure%20%28by%20ref%29) (SideEffectTriggerType :type EBlueprintableSideEffectTriggerType%20Enum%20%28by%20ref%29 :default "TriggeredAsTarget"))
  (Networking|SwitchHasAuthority
    (:Remote
      (switch Utilities|FlowControl|Switch|SwitchonGameplayTag SideEffectTriggerTag
        (:Quests.S36.SDA.BPSideEffect.B
          (Niagara|SpawnSystemAttached "/BattlePassS36_QuestActors/VFX/NS_Quest_ShardSign_Burst.NS_Quest_ShardSign_Burst" (Variables|Transformation|GetRootComponent) "None" "0.000000,0.000000,77.000000")
          (Components|Activation|Activate (Variables|Default|GetNS_Quest_ShardSign_Idle) false))))))

(fn ConstructionScript
  (|CallParentFunction))
